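# Translucent, unlit, two-sided material for Niagara mesh particles that draws a thickened spline streak with a traveling visible segment.
# Dynamic parameters drive SplineThicken width, texture pan speed and the position of a soft sphere mask along the UV.
# Opacity multiplies that mask, an end fade, a panning ribbon trail texture and particle alpha.
# Particle color becomes HDR emissive; low quality skips that conversion and depth fade.
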Material /BRCosmetics/Effects/Fort_Effects/Materials/Splines/M_Spline_Streak_Travel
graph {
 properties {
  bUsedWithNiagaraMeshParticles=true
  blendMode=BLEND_Translucent
  pixelDepthOffsetMode=PDOM_Legacy
  shadingModel=MSM_Unlit
  twoSided=true
  usedShadingModels=Unlit
 }
 nodes [
  Constant() {r=1}
  n1 = DynamicParameter() {defaultValue=<rgba 20 1 0.2 1>, paramNames [SplineWidth,PanSpeed,ZipTravel,TimeOffset]}
  MaterialFunctionCall(WidthBase=@n1.SplineWidth, WidthTip=@n1.SplineWidth) {materialFunction=&SplineThicken}
  n2 = ParticleColor()
  n3 = Clamp(None=@n2.RGB)
  n4 = ScalarParameter() {defaultValue=0.1, parameterName=HDRIntensity}
  n5 = ScalarParameter() {parameterName=DsaturateLuminance}
  MaterialFunctionCall(Emissive=QualitySwitch(
    Default=MaterialFunctionCall(LDR Input=@n3, HDR Tint=@n2.RGB, HDR Intensity Multiplier=@n4, Desaturate Luminance=@n5) {materialFunction=&DeriveHDRfromLDR}.RemappedRange
    Low=@n2.RGB
   )) {materialFunction=&MF_FX_EmissiveAdjustments}
  n6 = ComponentMask(None=TextureCoordinate()) {g=true}
  n7 = Power(Base=OneMinus(None=@n6))
  n8 = Constant() {r=0.5}
  n9 = ScalarParameter() {defaultValue=0.5, parameterName=EndFadeHardnes}
  n10 = ScalarParameter() {defaultValue=0.5, parameterName=ZipRadius}
  n11 = Multiply(A=@n10)
  n12 = Add(A=@n11)
  n13 = Subtract(B=@n11) {constA=0}
  n14 = Power(Base=SphereMask(A=@n6, B=LinearInterpolate(A=@n12, B=@n13, Alpha=@n1.ZipTravel) {constAlpha=1}, Radius=@n10) {hardnessPercent=0})
  n15 = VectorParameter() {parameterName=Tile, defaultValue=<rgba 1 0.7 0 0>}
  n16 = Multiply(A=TextureCoordinate(), B=AppendVector(A=@n15.R, B=@n15.R))
  n17 = MaterialFunctionCall(XY=@n16) {materialFunction=&Swizzle}
  n18 = ScalarParameter() {defaultValue=1, parameterName=SwizzleUVs}
  n19 = Constant()
  n20 = TextureSample(UVs=Panner(Coordinate=LinearInterpolate(A=@n16, B=@n17.YX, Alpha=@n18), Speed=AppendVector(A=@n1.PanSpeed, B=@n19))) {texture=/BRCosmetics/Effects/Fort_Effects/Textures/Ribbon/T_Ribbon_Trail_HardLineCurve}
  n21 = Multiply(A=@n2.A, B=Multiply(A=SphereMask(A=@n7, B=@n8, Radius=@n8, Hardness=@n9) {hardnessPercent=0}, B=Multiply(A=@n14, B=@n20.G)))
  QualitySwitch(Default=DepthFade(Opacity=@n21, FadeDistance=ScalarParameter() {defaultValue=20, parameterName=DepthFade}), Low=@n21)
 ]
}
